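# Physics asset for the hair on the F_MED_CrimpPerfume female character's FaceAcc part, using capsule bodies to simulate hair motion.
# Center, left and right five-bone chains, two-bone front strands and single-bone side pieces hang from base_hair bodies via limited swing and twist constraints.
# Free-angular cross constraints tie the left and right tips to the center tip.
# Head, neck, spine, pelvis, clavicle and upper-arm capsules provide body collision shapes.
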PhysicsAsset /BRCosmetics/Characters/Player/Female/Medium/Bodies/F_MED_CrimpPerfume/Meshes/Parts/F_MED_CrimpPerfume_FaceAcc_Physics
bodies {
 base_hair_c [{shapeType=Capsule, length=0.505, radius=0.505, shapeName=base_hair_c_capsule, rotation=<rot 0 -90 0>}]
 base_hair_front_l [{shapeType=Capsule, length=0.505, radius=0.505, shapeName=base_hair_front_l_capsule, rotation=<rot -90 0 0>}]
 base_hair_front_r [{shapeType=Capsule, length=0.505, radius=1, shapeName=base_hair_front_r_capsule, rotation=<rot -90 0 0>}]
 base_hair_l [{shapeType=Capsule, length=0.505, radius=0.505, shapeName=base_hair_l_capsule, rotation=<rot 0 -90 0>}]
 base_hair_r [{shapeType=Capsule, length=0.505, radius=0.505, shapeName=base_hair_r_capsule, rotation=<rot 0 -90 0>}]
 base_hair_side_l [{shapeType=Capsule, length=0.505, radius=0.505, shapeName=base_hair_side_l_capsule, rotation=<rot 0 -90 0>}]
 base_hair_side_r [{shapeType=Capsule, length=0.505, radius=0.505, shapeName=base_hair_side_r_capsule, rotation=<rot 0 -90 0>}]
 clavicle_l [{shapeType=Capsule, length=12, radius=5.5, shapeName=clavicle_l_capsule, center=<3.5 -1 -1>, rotation=<rot 80 -0.0003 -0.0003>}]
 clavicle_r [{shapeType=Capsule, length=12, radius=5.5, shapeName=clavicle_r_capsule, center=<-3.5 1 1>, rotation=<rot 80 -0.0003 -0.0004>}]
 dyn_hair_c_01 [{shapeType=Capsule, length=10, radius=5, shapeName=dyn_hair_c_01_capsule, center=<6 1 0>}]
 dyn_hair_c_02 [{shapeType=Capsule, length=10, radius=5, shapeName=dyn_hair_c_02_capsule, center=<6 0 0>}]
 dyn_hair_c_03 [{shapeType=Capsule, length=10, radius=5, shapeName=dyn_hair_c_03_capsule, center=<6 0 0>}]
 dyn_hair_c_04 [{shapeType=Capsule, length=10, radius=5, shapeName=dyn_hair_c_04_capsule, center=<6 2 0>}]
 dyn_hair_c_05 [{shapeType=Capsule, length=10, radius=6, shapeName=dyn_hair_c_05_capsule, center=<4 2 0>}]
 dyn_hair_front_01_l [{shapeType=Capsule, length=2, radius=4, shapeName=dyn_hair_front_01_l_capsule, center=<4 0 -3>}]
 dyn_hair_front_01_r [{shapeType=Capsule, length=4, radius=4, shapeName=dyn_hair_front_01_r_capsule, center=<4 0 1>, rotation=<rot 0 90 50>}]
 dyn_hair_front_02_l [{shapeType=Capsule, length=2, radius=3, shapeName=dyn_hair_front_02_l_capsule, center=<2 -1 1.5>}]
 dyn_hair_front_02_r [{shapeType=Capsule, length=6, radius=3, shapeName=dyn_hair_front_02_r_capsule, center=<0.5 1 1>, rotation=<rot 9.3913 3.4512 20.2836>}]
 dyn_hair_l_01 [{shapeType=Capsule, length=10, radius=5, shapeName=dyn_hair_l_01_capsule, center=<6 0 -2>, rotation=<rot 0 0 20>}]
 dyn_hair_l_02 [{shapeType=Capsule, length=10, radius=5, shapeName=dyn_hair_l_02_capsule, center=<6 -1 2>, rotation=<rot 0 0 10>}]
 dyn_hair_l_03 [{shapeType=Capsule, length=10, radius=5, shapeName=dyn_hair_l_03_capsule, center=<6 0 2>}]
 dyn_hair_l_04 [{shapeType=Capsule, length=10, radius=5, shapeName=dyn_hair_l_04_capsule, center=<6 2 2>}]
 dyn_hair_l_05 [{shapeType=Capsule, length=10, radius=6, shapeName=dyn_hair_l_05_capsule, center=<4 2 2>, rotation=<rot 0 0 10>}]
 dyn_hair_r_01 [{shapeType=Capsule, length=10, radius=5, shapeName=dyn_hair_r_01_capsule, center=<6 0 2>, rotation=<rot 0 0 -20>}]
 dyn_hair_r_02 [{shapeType=Capsule, length=10, radius=5, shapeName=dyn_hair_r_02_capsule, center=<6 -1 -2>, rotation=<rot 0 0 -10>}]
 dyn_hair_r_03 [{shapeType=Capsule, length=10, radius=5, shapeName=dyn_hair_r_03_capsule, center=<6 0 -2>}]
 dyn_hair_r_04 [{shapeType=Capsule, length=10, radius=5, shapeName=dyn_hair_r_04_capsule, center=<6 2 -2>}]
 dyn_hair_r_05 [{shapeType=Capsule, length=10, radius=6, shapeName=dyn_hair_r_05_capsule, center=<4 2 -2>, rotation=<rot 0 0 -10>}]
 dyn_hair_side_01_l [{shapeType=Capsule, length=5, radius=4, shapeName=dyn_hair_side_01_l_capsule, center=<6 0 0>, rotation=<rot 0 90 90>}]
 dyn_hair_side_01_r [{shapeType=Capsule, length=5, radius=4, shapeName=dyn_hair_side_01_r_capsule, center=<6 0 0>, rotation=<rot 0 90 90>}]
 head [{shapeType=Capsule, length=7, radius=6, shapeName=head_capsule, center=<3 3 0>, rotation=<rot 90 -10.0001 0>}]
 neck_01 [{shapeType=Capsule, length=8, radius=4.5, shapeName=neck_01_capsule, center=<3.6683 -0.2876 0>, rotation=<rot 90 0 0>}]
 neck_02 [{shapeType=Capsule, length=3.5143, radius=4.4513, shapeName=neck_02_capsule, center=<2 0 0>, rotation=<rot -90 0 0>}]
 pelvis [{shapeType=Capsule, length=10, radius=11, shapeName=pelvis_capsule, center=<-4 -2 0>}]
 spine_01 [{shapeType=Capsule, length=10, radius=8, shapeName=spine_01_capsule}]
 spine_02 [{shapeType=Capsule, length=6, radius=8, shapeName=spine_02_capsule, center=<0 0.1555 0>}]
 spine_03 [{shapeType=Capsule, length=15, radius=9, shapeName=spine_04_capsule, center=<0 2.5 0>, rotation=<rot 90 0 10.0001>}]
 spine_04 [{shapeType=Capsule, length=10, radius=10, shapeName=spine_04_capsule, center=<0 2 0>}]
 spine_05 [
  {shapeType=Capsule, length=12, radius=9.25, shapeName=spine_05_capsule, center=<-2 1 0>}
  {shapeType=Capsule, length=12, radius=7.75, shapeName=spine_05_capsule, center=<-6.4845 2.8778 0>}
 ]
 upperarm_l [{shapeType=Capsule, length=20, radius=3.5, shapeName=upperarm_l_capsule, center=<12 -0.5 0>, rotation=<rot 90 0 0>}]
 upperarm_r [{shapeType=Capsule, length=20, radius=3.5, shapeName=upperarm_r_capsule, center=<-12 0.5 0>, rotation=<rot 90 0 0>}]
}
constraints [
 {bone1Name=dyn_hair_front_02_r, bone2Name=dyn_hair_front_01_r, swing1LimitDegrees=25, swing2LimitDegrees=25, twistLimitDegrees=2}
 {bone1Name=dyn_hair_front_02_l, bone2Name=dyn_hair_front_01_l, swing1LimitDegrees=25, swing2LimitDegrees=25, twistLimitDegrees=3}
 {bone1Name=dyn_hair_c_02, bone2Name=dyn_hair_c_01, swing1LimitDegrees=25, swing2LimitDegrees=15, twistLimitDegrees=15}
 {bone1Name=dyn_hair_c_03, bone2Name=dyn_hair_c_02, swing1LimitDegrees=25, swing2LimitDegrees=15, twistLimitDegrees=15}
 {bone1Name=dyn_hair_c_04, bone2Name=dyn_hair_c_03, swing1LimitDegrees=25, swing2LimitDegrees=15, twistLimitDegrees=1}
 {bone1Name=dyn_hair_c_05, bone2Name=dyn_hair_c_04, swing1LimitDegrees=25, swing2LimitDegrees=25, twistLimitDegrees=1}
 {bone1Name=dyn_hair_l_02, bone2Name=dyn_hair_l_01, swing1LimitDegrees=25, swing2LimitDegrees=15, twistLimitDegrees=15}
 {bone1Name=dyn_hair_l_03, bone2Name=dyn_hair_l_02, swing1LimitDegrees=25, swing2LimitDegrees=15, twistLimitDegrees=15}
 {bone1Name=dyn_hair_l_04, bone2Name=dyn_hair_l_03, swing1LimitDegrees=25, swing2LimitDegrees=15, twistLimitDegrees=1}
 {bone1Name=dyn_hair_l_05, bone2Name=dyn_hair_l_04, swing1LimitDegrees=25, swing2LimitDegrees=25, twistLimitDegrees=1}
 {bone1Name=dyn_hair_r_02, bone2Name=dyn_hair_r_01, swing1LimitDegrees=25, swing2LimitDegrees=15, twistLimitDegrees=15}
 {bone1Name=dyn_hair_r_03, bone2Name=dyn_hair_r_02, swing1LimitDegrees=25, swing2LimitDegrees=15, twistLimitDegrees=15}
 {bone1Name=dyn_hair_r_04, bone2Name=dyn_hair_r_03, swing1LimitDegrees=25, swing2LimitDegrees=15, twistLimitDegrees=1}
 {bone1Name=dyn_hair_r_05, bone2Name=dyn_hair_r_04, swing1LimitDegrees=25, swing2LimitDegrees=25, twistLimitDegrees=1}
 {bone1Name=dyn_hair_side_01_l, bone2Name=base_hair_side_l, swing1LimitDegrees=15, swing2LimitDegrees=15, twistLimitDegrees=1}
 {bone1Name=dyn_hair_side_01_r, bone2Name=base_hair_side_r, swing1LimitDegrees=15, swing2LimitDegrees=15, twistLimitDegrees=1}
 {bone1Name=dyn_hair_c_01, bone2Name=base_hair_c, swing1LimitDegrees=25, swing2LimitDegrees=15, twistLimitDegrees=15}
 {bone1Name=dyn_hair_l_01, bone2Name=base_hair_l, swing1LimitDegrees=25, swing2LimitDegrees=15, twistLimitDegrees=15}
 {bone1Name=dyn_hair_r_01, bone2Name=base_hair_r, swing1LimitDegrees=25, swing2LimitDegrees=15, twistLimitDegrees=15}
 {bone1Name=dyn_hair_front_01_l, bone2Name=base_hair_front_l, swing1LimitDegrees=25, swing2LimitDegrees=25, twistLimitDegrees=3}
 {bone1Name=dyn_hair_front_01_r, bone2Name=base_hair_front_r, swing1LimitDegrees=25, swing2LimitDegrees=25, twistLimitDegrees=3}
 {
  bone1Name=dyn_hair_r_05
  bone2Name=dyn_hair_c_05
  swing1LimitDegrees=45
  swing1Motion=Free
  swing2LimitDegrees=45
  swing2Motion=Free
  twistLimitDegrees=45
  twistMotion=Free
 }
 {
  bone1Name=dyn_hair_l_05
  bone2Name=dyn_hair_c_05
  swing1LimitDegrees=45
  swing1Motion=Free
  swing2LimitDegrees=45
  swing2Motion=Free
  twistLimitDegrees=45
  twistMotion=Free
 }
]
