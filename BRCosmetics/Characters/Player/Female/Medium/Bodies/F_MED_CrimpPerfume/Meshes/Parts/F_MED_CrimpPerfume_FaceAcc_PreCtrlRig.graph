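# Control Rig for the CrimpPerfume female face accessory that repositions base hair bones, parents of the dyn_hair chains, as the head moves.
# A sequence first rotation-constrains both front hair bones evenly between spine_05 and head.
# Four spherical pose readers then track head orientation against neck or upper spine, weighting corrective rotations on center, left, right, side and front hair.
# Constraints for base_hair_c, _l and _r are unconnected and never execute.
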ControlRigBlueprint /BRCosmetics/Characters/Player/Female/Medium/Bodies/F_MED_CrimpPerfume/Meshes/Parts/F_MED_CrimpPerfume_FaceAcc_PreCtrlRig
graphs [
 RigVMFunctionLibrary {links [], nodes []}
 RigVMFunction_Sequence_1_ContainedGraph {nodes [
   n1 = RigVMFunctionEntryNode()
   RigVMFunction_Sequence = &RigVMFunction_Sequence(ExecuteContext=@n1.ExecuteContext)
   n2 = &RigVMFunction_Sequence(ExecuteContext=@RigVMFunction_Sequence.B)
   n3 = &RigVMFunction_Sequence(ExecuteContext=@n2.B)
   RigVMFunctionReturnNode(A=@RigVMFunction_Sequence.A, B=@n2.A, C=@n3.A, D=@n3.B)
  ]}
 RigVMModel {nodes [
   RigUnit_BeginExecution = &RigUnit_BeginExecution()
   RigVMFunction_Sequence = RigVMAggregateNode(ExecuteContext=@RigUnit_BeginExecution.ExecutePin)
   &RigUnit_RotationConstraintLocalSpaceOffset(
    AdvancedSettings=(RotationOrderForFilter=XZY)
    Child=(Type=Bone,Name="base_hair_front_l")
    Filter=(bX=True,bY=True,bZ=True)
    Parents=((Item=(Type=Bone,Name="spine_05"),Weight=0.5),(Item=(Type=Bone,Name="head"),Weight=0.5))
    Weight=1
    bMaintainOffset=true
    ExecutePin=@RigVMFunction_Sequence.A
   )
   > &RigUnit_RotationConstraintLocalSpaceOffset(
    AdvancedSettings=(RotationOrderForFilter=XZY)
    Child=(Type=Bone,Name="base_hair_front_r")
    Filter=(bX=True,bY=True,bZ=True)
    Parents=((Item=(Type=Bone,Name="spine_05"),Weight=0.5),(Item=(Type=Bone,Name="head"),Weight=0.5))
    Weight=1
    bMaintainOffset=true
   )
   n1 = &RigUnit_SphericalPoseReader(
    ActiveRegionScaleFactors=(PositiveWidth=1,NegativeWidth=1,PositiveHeight=1,NegativeHeight=1)
    ActiveRegionSize=0.1
    Debug=(bDrawDebug=false,DebugScale=25,DebugSegments=20,DebugThickness=0.25)
    DriverAxis=(X=0,Y=-1,Z=0)
    DriverItem=(Type=Bone,Name="head")
    FalloffRegionScaleFactors=(PositiveWidth=2,NegativeWidth=2,PositiveHeight=2,NegativeHeight=2)
    FalloffSize=0.2
    FlipWidthScaling=false
    OptionalParentItem=(Type=Bone,Name="neck_01")
    RotationOffset=(X=-90,Y=0,Z=-55)
    ExecutePin=@RigVMFunction_Sequence.B
   )
   > &RigUnit_ModifyTransforms(
    ItemToModify=((Item=(Type=Bone,Name="base_hair_c"),Transform=(Rotation=(X=0,Y=0,Z=-0.5373,W=0.8434),Translation=(X=0,Y=0,Z=0),Scale3D=(X=1,Y=1,Z=1))))
    Weight=@n1.OutputParam
   )
   > &RigUnit_ModifyTransforms(
    ItemToModify=((Item=(Type=Bone,Name="base_hair_l"),Transform=(Rotation=(X=0,Y=0,Z=-0.5373,W=0.8434),Translation=(X=0,Y=0,Z=0),Scale3D=(X=1,Y=1,Z=1))))
    Weight=@n1.OutputParam
   )
   > &RigUnit_ModifyTransforms(
    ItemToModify=((Item=(Type=Bone,Name="base_hair_r"),Transform=(Rotation=(X=0,Y=0,Z=-0.5373,W=0.8434),Translation=(X=0,Y=0,Z=0),Scale3D=(X=1,Y=1,Z=1))))
    Weight=@n1.OutputParam
   )
   > &RigUnit_ModifyTransforms(
    ItemToModify=((Item=(Type=Bone,Name="base_hair_side_l"),Transform=(Rotation=(X=0,Y=0,Z=-0.5373,W=0.8434),Translation=(X=0,Y=0,Z=0),Scale3D=(X=1,Y=1,Z=1))))
    Weight=@n1.OutputParam
   )
   > &RigUnit_ModifyTransforms(
    ItemToModify=((Item=(Type=Bone,Name="base_hair_side_r"),Transform=(Rotation=(X=0,Y=0,Z=-0.5373,W=0.8434),Translation=(X=0,Y=0,Z=0),Scale3D=(X=1,Y=1,Z=1))))
    Weight=@n1.OutputParam
   )
   > n2 = &RigUnit_SphericalPoseReader(
    ActiveRegionScaleFactors=(PositiveWidth=0.25,NegativeWidth=0.25,PositiveHeight=1,NegativeHeight=1)
    ActiveRegionSize=0.1
    Debug=(bDrawDebug=false,DebugScale=25,DebugSegments=20,DebugThickness=0.25)
    DriverAxis=(X=0,Y=1,Z=0)
    DriverItem=(Type=Bone,Name="head")
    FalloffRegionScaleFactors=(PositiveWidth=0.75,NegativeWidth=0.5,PositiveHeight=4,NegativeHeight=4)
    FalloffSize=0.2
    OptionalParentItem=(Type=Bone,Name="spine_05")
    RotationOffset=(X=90,Y=0,Z=25)
   )
   > &RigUnit_ModifyTransforms(
    ItemToModify=((Item=(Type=Bone,Name="base_hair_c"),Transform=(Rotation=(X=0,Y=0,Z=0.1305,W=0.9914),Translation=(X=0,Y=0,Z=0),Scale3D=(X=1,Y=1,Z=1))))
    Weight=@n2.OutputParam
   )
   > &RigUnit_ModifyTransforms(
    ItemToModify=((Item=(Type=Bone,Name="base_hair_l"),Transform=(Rotation=(X=0,Y=0,Z=0.1305,W=0.9914),Translation=(X=0,Y=0,Z=0),Scale3D=(X=1,Y=1,Z=1))))
    Weight=@n2.OutputParam
   )
   > &RigUnit_ModifyTransforms(
    ItemToModify=((Item=(Type=Bone,Name="base_hair_r"),Transform=(Rotation=(X=0,Y=0,Z=0.1305,W=0.9914),Translation=(X=0,Y=0,Z=0),Scale3D=(X=1,Y=1,Z=1))))
    Weight=@n2.OutputParam
   )
   > &RigUnit_ModifyTransforms(
    ItemToModify=((Item=(Type=Bone,Name="base_hair_front_l"),Transform=(Rotation=(X=0,Y=0,Z=-0.1305,W=0.9914),Translation=(X=0,Y=0,Z=0),Scale3D=(X=1,Y=1,Z=1))))
    Weight=@n2.OutputParam
   )
   > &RigUnit_ModifyTransforms(
    ItemToModify=((Item=(Type=Bone,Name="base_hair_front_r"),Transform=(Rotation=(X=0,Y=0,Z=-0.1305,W=0.9914),Translation=(X=0,Y=0,Z=0),Scale3D=(X=1,Y=1,Z=1))))
    Weight=@n2.OutputParam
   )
   > n3 = &RigUnit_SphericalPoseReader(
    ActiveRegionScaleFactors=(PositiveWidth=1,NegativeWidth=0.5,PositiveHeight=1,NegativeHeight=1)
    ActiveRegionSize=0.1
    Debug=(bDrawDebug=false,DebugScale=25,DebugSegments=20,DebugThickness=0.25)
    DriverAxis=(X=0,Y=-1,Z=0)
    DriverItem=(Type=Bone,Name="head")
    FalloffRegionScaleFactors=(PositiveWidth=1,NegativeWidth=0.5,PositiveHeight=1,NegativeHeight=1)
    FalloffSize=0.25
    OptionalParentItem=(Type=Bone,Name="neck_02")
    RotationOffset=(X=-55,Y=0,Z=-10)
   )
   > &RigUnit_ModifyTransforms(
    ItemToModify=((Item=(Type=Bone,Name="base_hair_r"),Transform=(Rotation=(X=0,Y=0,Z=-0.2164,W=0.9763),Translation=(X=0,Y=0,Z=0),Scale3D=(X=1,Y=1,Z=1))))
    Weight=@n3.OutputParam
   )
   > &RigUnit_ModifyTransforms(
    ItemToModify=((Item=(Type=Bone,Name="base_hair_front_r"),Transform=(Rotation=(X=0,Y=0,Z=0.1305,W=0.9914),Translation=(X=0,Y=0,Z=0),Scale3D=(X=1,Y=1,Z=1))))
    Weight=@n3.OutputParam
   )
   > &RigUnit_ModifyTransforms(
    ItemToModify=((Item=(Type=Bone,Name="base_hair_c"),Transform=(Rotation=(X=0,Y=0,Z=-0.1305,W=0.9914),Translation=(X=0,Y=0,Z=0),Scale3D=(X=1,Y=1,Z=1))))
    Weight=@n3.OutputParam
   )
   > n4 = &RigUnit_SphericalPoseReader(
    ActiveRegionScaleFactors=(PositiveWidth=1,NegativeWidth=0.5,PositiveHeight=1,NegativeHeight=1)
    ActiveRegionSize=0.1
    Debug=(bDrawDebug=false,DebugScale=25,DebugSegments=20,DebugThickness=0.25)
    DriverAxis=(X=0,Y=-1,Z=0)
    DriverItem=(Type=Bone,Name="head")
    FalloffRegionScaleFactors=(PositiveWidth=1,NegativeWidth=0.5,PositiveHeight=1,NegativeHeight=1)
    FalloffSize=0.25
    OptionalParentItem=(Type=Bone,Name="neck_02")
    RotationOffset=(X=235,Y=0,Z=-10)
   )
   > &RigUnit_ModifyTransforms(
    ItemToModify=((Item=(Type=Bone,Name="base_hair_l"),Transform=(Rotation=(X=0,Y=0,Z=-0.2164,W=0.9763),Translation=(X=0,Y=0,Z=0),Scale3D=(X=1,Y=1,Z=1))))
    Weight=@n4.OutputParam
   )
   > &RigUnit_ModifyTransforms(
    ItemToModify=((Item=(Type=Bone,Name="base_hair_front_l"),Transform=(Rotation=(X=0,Y=0,Z=0.1305,W=0.9914),Translation=(X=0,Y=0,Z=0),Scale3D=(X=1,Y=1,Z=1))))
    Weight=@n4.OutputParam
   )
   > &RigUnit_ModifyTransforms(
    ItemToModify=((Item=(Type=Bone,Name="base_hair_c"),Transform=(Rotation=(X=0,Y=0,Z=-0.1305,W=0.9914),Translation=(X=0,Y=0,Z=0),Scale3D=(X=1,Y=1,Z=1))))
    Weight=@n4.OutputParam
   )
   &RigUnit_RotationConstraintLocalSpaceOffset(
    AdvancedSettings=(RotationOrderForFilter=XZY)
    Child=(Type=Bone,Name="base_hair_c")
    ExecutePin=()
    Filter=(bX=True,bY=True,bZ=True)
    Parents=((Item=(Type=Bone,Name="spine_05"),Weight=0.5),(Item=(Type=Bone,Name="head"),Weight=0.5))
    Weight=1
   )
   > &RigUnit_RotationConstraintLocalSpaceOffset(
    AdvancedSettings=(RotationOrderForFilter=XZY)
    Child=(Type=Bone,Name="base_hair_l")
    Filter=(bX=True,bY=True,bZ=True)
    Parents=((Item=(Type=Bone,Name="spine_05"),Weight=0.5),(Item=(Type=Bone,Name="head"),Weight=0.5))
    Weight=1
   )
   > &RigUnit_RotationConstraintLocalSpaceOffset(
    AdvancedSettings=(RotationOrderForFilter=XZY)
    Child=(Type=Bone,Name="base_hair_r")
    Filter=(bX=True,bY=True,bZ=True)
    Parents=((Item=(Type=Bone,Name="spine_05"),Weight=0.5),(Item=(Type=Bone,Name="head"),Weight=0.5))
    Weight=1
    bMaintainOffset=true
   )
  ]}
]
hierarchy [
 root {loc=<0 0 0>, rot=<0 0 0 1>}
 pelvis {parent=root, loc=<0 -0.5301 98.0571>, rot=<0.0127 0.707 -0.0127 -0.707>}
 spine_01 {parent=pelvis, loc=<6.1208 -0.9328 0>, rot=<0 0 0.034 -0.9994>}
 spine_02 {parent=spine_01, loc=<6.5997 -0.2095 0>, rot=<0 0 0.0238 -0.9997>}
 spine_03 {parent=spine_02, loc=<6.4914 -0.1237 0>, rot=<0 0 0.0149 -0.9999>}
 spine_04 {parent=spine_03, loc=<6.9598 0.0794 0>, rot=<0 0 -0.0132 -0.9999>}
 spine_05 {parent=spine_04, loc=<6.9987 0 0>, rot=<0 0 -0.0015 -1>}
 neck_01 {parent=spine_05, loc=<11.3138 0.1529 0>, rot=<0 0 -0.2003 -0.9797>}
 neck_02 {parent=neck_01, loc=<4.5627 0 0>, rot=<0 0 0.0504 -0.9987>}
 head {parent=neck_02, loc=<5.3914 0 0>, rot=<0 0 0.115 -0.9934>}
 base_hair_side_l {parent=head, loc=<4.0101 -2.1342 -7.6847>, rot=<0.2458 -0.9412 -0.2234 -0.0616>}
 dyn_hair_side_01_l {parent=base_hair_side_l, loc=<0 0 0>, rot=<0 0 0 1>}
 base_hair_r {parent=head, loc=<4.6373 -7.1077 5.724>, rot=<0.1516 -0.9884 0 0>}
 dyn_hair_r_01 {parent=base_hair_r, loc=<0 0 0>, rot=<0 0 0 1>}
 dyn_hair_r_02 {parent=dyn_hair_r_01, loc=<10 0 0>, rot=<0 0 0 1>}
 dyn_hair_r_03 {parent=dyn_hair_r_02, loc=<10 0 0>, rot=<0 0 -0.0343 -0.9994>}
 dyn_hair_r_04 {parent=dyn_hair_r_03, loc=<10 0 0>, rot=<0 0 -0.106 -0.9944>}
 dyn_hair_r_05 {parent=dyn_hair_r_04, loc=<10 0 0>, rot=<0 0 0 1>}
 base_hair_l {parent=head, loc=<4.6613 -7.2984 -5.7372>, rot=<0.0627 -0.998 0 0>}
 dyn_hair_l_01 {parent=base_hair_l, loc=<0 0 0>, rot=<0 0 0.0754 -0.9972>}
 dyn_hair_l_02 {parent=dyn_hair_l_01, loc=<10 0 0>, rot=<0 0 0 1>}
 dyn_hair_l_03 {parent=dyn_hair_l_02, loc=<10 0 0>, rot=<0 0 -0.0601 -0.9982>}
 dyn_hair_l_04 {parent=dyn_hair_l_03, loc=<10 0 0>, rot=<0 0 -0.0661 -0.9978>}
 dyn_hair_l_05 {parent=dyn_hair_l_04, loc=<10 0 0>, rot=<0 0 0 1>}
 base_hair_c {parent=head, loc=<5.0721 -10.7594 0>, rot=<0.0627 -0.998 0 0>}
 dyn_hair_c_01 {parent=base_hair_c, loc=<0 0 0>, rot=<0 0 0 1>}
 dyn_hair_c_02 {parent=dyn_hair_c_01, loc=<10 0 0>, rot=<0 0 0 1>}
 dyn_hair_c_03 {parent=dyn_hair_c_02, loc=<10 0 0>, rot=<0 0 0 1>}
 dyn_hair_c_04 {parent=dyn_hair_c_03, loc=<10 0 0>, rot=<0 0 0 1>}
 dyn_hair_c_05 {parent=dyn_hair_c_04, loc=<10 0 0>, rot=<0 0 0 1>}
 base_hair_side_r {parent=head, loc=<3.7083 0.2593 7.6847>, rot=<0.2604 -0.9461 0.1845 0.0549>}
 dyn_hair_side_01_r {parent=base_hair_side_r, loc=<0 0 0>, rot=<0 0.0094 0.0138 -0.9999>}
 base_hair_front_l {parent=head, loc=<-2.1874 -3.1864 -6.2072>, rot=<-0.0281 0.2283 0.9525 0.1994>}
 dyn_hair_front_01_l {parent=base_hair_front_l, loc=<0 0 0>, rot=<0 0 0 1>}
 dyn_hair_front_02_l {parent=dyn_hair_front_01_l, loc=<7.4 0 0>, rot=<0 0 0 1>}
 dyn_hair_front_b_01_l {parent=dyn_hair_front_01_l, loc=<3.1044 0.3919 -3.036>, rot=<-0.35 -0.3667 0.0974 -0.8564>}
 base_hair_front_r {parent=head, loc=<-2.2987 -2.1419 4.7271>, rot=<0.2593 -0.0862 0.9385 0.2111>}
 dyn_hair_front_01_r {parent=base_hair_front_r, loc=<0 0 0>, rot=<0 0 0 1>}
 dyn_hair_front_02_r {parent=dyn_hair_front_01_r, loc=<7 0 0>, rot=<0 0 0 1>}
 base_hair_front_top_r {parent=head, loc=<2.9774 4.7354 6.7398>, rot=<0.4653 -0.524 0.7133 -0.0093>}
 dyn_hair_front_top_01_r {parent=base_hair_front_top_r, loc=<0 0 0>, rot=<0 0 0 1>}
 dyn_hair_front_top_02_r {parent=dyn_hair_front_top_01_r, loc=<5 0 0>, rot=<0 0 0 1>}
 base_hair_front_top_l {parent=head, loc=<11.8589 7.6919 -2.3829>, rot=<0.4439 -0.6312 -0.5187 -0.3681>}
 dyn_hair_front_top_01_l {parent=base_hair_front_top_l, loc=<0 0 0>, rot=<0 0 0 1>}
 dyn_hair_front_top_02_l {parent=dyn_hair_front_top_01_l, loc=<4.0377 0 0>, rot=<0.0113 -0.0255 -0.4046 -0.9141>}
 dyn_hair_front_top_03_l {parent=dyn_hair_front_top_02_l, loc=<5 0 0>, rot=<-0.0236 -0.1788 0.1286 -0.9752>}
 base_hair_flower {parent=head, loc=<6.2502 2.4157 8.5977>, rot=<-0.4403 0.3739 -0.585 -0.5693>}
 dyn_hair_flower_01 {parent=base_hair_flower, loc=<0.8313 0 0>, rot=<0 0 0 1>}
 dyn_hair_flower_02 {parent=base_hair_flower, loc=<1.7478 0 0>, rot=<0 0 0 1>}
 dyn_hair_flower_03 {parent=base_hair_flower, loc=<0.9077 0 0>, rot=<0 0 0 1>}
 head_null {parent=neck_02, type=Null, loc=<5.3914 0 0>, rot=<0 0 0.115 -0.9934>}
 clavicle_l {parent=spine_05, loc=<5.4831 -0.1053 -3.2946>, rot=<0.0465 -0.7289 -0.009 -0.6829>}
 upperarm_l {parent=clavicle_l, loc=<9.8391 -1.4584 -0.2012>, rot=<-0.0087 -0.3111 -0.0215 -0.9501>}
 lowerarm_l {parent=upperarm_l, loc=<26.465 -0.8345 0.259>, rot=<0.0036 0.0161 -0.2298 -0.9731>}
 clavicle_r {parent=spine_05, loc=<5.4831 -0.1053 3.2946>, rot=<0.7289 0.0465 -0.6829 0.009>}
 upperarm_r {parent=clavicle_r, loc=<-9.8658 1.2136 0.4105>, rot=<0.0164 -0.3245 -0.0269 -0.9454>}
 lowerarm_r {parent=upperarm_r, loc=<-26.465 0.8345 -0.259>, rot=<0.0036 0.0161 -0.2298 -0.9731>}
 thigh_l_null {parent=pelvis, type=Null, loc=<-1.0855 -0.2109 -8.7053>, rot=<0.0009 0.0355 0.0155 -0.9993>}
 thigh_r_null {parent=pelvis, type=Null, loc=<-1.0855 -0.2109 8.7053>, rot=<0.0355 -0.0009 0.9993 0.0155>}
 Mstr_Skirt_Mid_Bck_null {parent=pelvis, type=Null, loc=<3.8313 -12.246 0>, rot=<0.0276 -0.7066 0.7066 -0.0276>}
 Mstr_Skirt_Bck_l_null {parent=pelvis, type=Null, loc=<3.9552 -8.7914 -8.7937>, rot=<0.023 -0.903 0.4222 -0.0765>}
 Mstr_Skirt_Frnt_l_null {parent=pelvis, type=Null, loc=<4.444 4.5271 -8.8579>, rot=<-0.0227 -0.8844 -0.4551 -0.1009>}
 Mstr_Skirt_Side_l_null {parent=pelvis, type=Null, loc=<4.1987 -1.9972 -13.4728>, rot=<-0.0073 -0.9891 0.0886 -0.1171>}
]
